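# Unlit translucent material for the PaddleGinger camera-lens effect on Niagara mesh particles, drawn without depth testing.
# Three U-offset samples of panning smoke form RGB for color fringing, then get sharpened, tinted cyan and added to a blue gradient driven by a panning triangle texture.
# Particle color scales the emissive, which passes through MF_FX_EmissiveAdjustments.
# Opacity multiplies particle alpha, a V-gradient band mask and parameterized scanlines from MF_Atlas_Scanlines_01.
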Material /MotherGameplay/Items/PaddleGinger/Assets/FX/M_Camera_Lens_PaddleGinger
graph {
 properties {
  bDisableDepthTest=true
  bUsedWithNiagaraMeshParticles=true
  blendMode=BLEND_Translucent
  pixelDepthOffsetMode=PDOM_Legacy
  shadingModel=MSM_Unlit
  usedShadingModels=Unlit
 }
 nodes [
  n1 = ParticleColor()
  n2 = Constant() {r=0.005}
  n3 = Add(A=TextureCoordinate() {uTiling=5}, B=@n2)
  n4 = TextureCoordinate() {uTiling=5}
  n5 = Subtract(A=TextureCoordinate() {uTiling=5}, B=@n2)
  n6 = MaterialFunctionCall(
   X=TextureSample(UVs=Panner(Coordinate=@n3) {speedY=-0.5}) {texture=&T_Inky_Smoke_Tile}.R
   Y=TextureSample(UVs=Panner(Coordinate=@n4) {speedY=-0.5}) {texture=&T_Inky_Smoke_Tile}.R
   Z=TextureSample(UVs=Panner(Coordinate=@n5) {speedY=-0.5}) {texture=&T_Inky_Smoke_Tile}.R
  ) {materialFunction=&MakeFloat3}
  n7 = VectorParameter() {parameterName=Color01, defaultValue=<rgba 0.0019 0.0289 0.0521 1>}
  n8 = VectorParameter() {parameterName=Color02, defaultValue=<rgba 0 0.2102 1 1>}
  n9 = TextureCoordinate() {uTiling=7, vTiling=0}
  n10 = Multiply(
   A=Constant() {r=0.5}
   B=TextureSample(UVs=Panner(Coordinate=@n9) {speedY=-0.5}) {texture=/Game/Gadgets/Assets/VinderTech_GliderChute/Glider_Ashton_Salt_Lake/FX/T_AshtonSaltLake_JumpTri}.R
  )
  n11 = Add(
   A=Multiply(A=Power(Base=@n6.Result) {constExponent=3}, B=Constant3Vector() {constant=<rgba 1.6177 13 12.612 1>})
   B=Multiply(A=LinearInterpolate(A=@n7.RGB, B=@n8.RGB, Alpha=@n10)) {constB=10}
  )
  MaterialFunctionCall(Emissive=Multiply(A=@n1.RGB, B=Multiply(A=@n11))) {materialFunction=&MF_FX_EmissiveAdjustments}
  n12 = ComponentMask(None=TextureCoordinate()) {g=true}
  n13 = Power(Base=ComponentMask(None=TextureCoordinate()) {g=true}) {constExponent=12}
  n14 = Multiply(A=Power(Base=OneMinus(None=@n12)), B=Saturate(None=Multiply(A=@n13) {constB=30}))
  Multiply(
   A=Multiply(A=@n1.A, B=Multiply(A=@n14) {constB=20})
   B=MaterialFunctionCall!(
    Number of Scanlines=ScalarParameter() {defaultValue=50, parameterName=Scanline Amount}
    Scanline Speed
    "Scanlines Thickness (Bias)"=ScalarParameter() {defaultValue=-0.15, parameterName=ScanlineSize}
    Scanlines Dark Value=ScalarParameter() {defaultValue=0.5, parameterName=Scanlines Dark}
    Scanlines Light Value=ScalarParameter() {defaultValue=1, parameterName=Scanlines Light}
    Block Opacity
    UVs
    Time
    Block Repeat
   ) {materialFunction=/Game/Effects/Fort_Effects/Effects/Atlas/Materials/MF_Atlas_Scanlines_01}.Result
  )
 ]
}
