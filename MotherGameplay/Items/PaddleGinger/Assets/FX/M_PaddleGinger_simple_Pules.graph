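# Translucent, unlit, two-sided pulse effect material for the PaddleGinger item, usable on Niagara mesh particles.
# Particle color feeds an emissive adjustment function, while particle alpha scales a soft V-axis band mask.
# The band is multiplied by scanlines and a gradient that the Mask_Offset dynamic parameter slides along V, with Use_simple_mask blending back to the plain band.
# A shading path switch drops the depth fade on forward and mobile.
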Material /MotherGameplay/Items/PaddleGinger/Assets/FX/M_PaddleGinger_simple_Pules
graph {
 properties {
  bUsedWithNiagaraMeshParticles=true
  blendMode=BLEND_Translucent
  pixelDepthOffsetMode=PDOM_Legacy
  shadingModel=MSM_Unlit
  twoSided=true
  usedShadingModels=Unlit
 }
 nodes [
  MaterialFunctionCall() {materialFunction=&DebugTimeSine}
  n1 = ParticleColor()
  MaterialFunctionCall(Emissive=@n1.RGB) {materialFunction=&MF_FX_EmissiveAdjustments}
  n2 = ComponentMask(None=TextureCoordinate()) {g=true}
  n3 = Power(Base=@n2, Exp=ScalarParameter() {defaultValue=1.3, parameterName=Power}) {constExponent=1.5}
  n4 = Saturate(None=OneMinus(None=@n2))
  n5 = Multiply(A=@n1.A, B=Saturate(None=Multiply(A=@n3, B=@n4)))
  n6 = MaterialFunctionCall!(
   Number of Scanlines=ScalarParameter() {defaultValue=50, parameterName=Scanline Amount}
   Scanline Speed
   "Scanlines Thickness (Bias)"=ScalarParameter() {defaultValue=-0.15, parameterName=ScanlineSize}
   Scanlines Dark Value=ScalarParameter() {defaultValue=0.5, parameterName=Scanlines Dark}
   Scanlines Light Value=ScalarParameter() {defaultValue=1, parameterName=Scanlines Light}
   Block Opacity
   UVs
   Time
   Block Repeat
  ) {materialFunction=/Game/Effects/Fort_Effects/Effects/Atlas/Materials/MF_Atlas_Scanlines_01}
  n7 = Panner(
   Coordinate=TextureCoordinate() {vTiling=0.7}
   Time=DynamicParameter() {defaultValue=<rgba 0.5 1 1 1>, paramNames [Mask_Offset,Param2,Param3,Param4]}.Mask_Offset
  ) {speedY=1}
  n8 = LinearInterpolate(
   A=Multiply(
    A=Multiply(A=@n5, B=@n6.Result)
    B=TextureSample(UVs=@n7) {samplerSource=SSM_Clamp_WorldGroupSettings, texture=/Game/Effects/Fort_Effects/Textures/Grad/gradient_healed}.R
   )
   B=@n5
   Alpha=ScalarParameter() {parameterName=Use_simple_mask}
  )
  n9 = DepthFade(Opacity=@n8, FadeDistance=ScalarParameter() {defaultValue=50, parameterName=Depth_Fade})
  ShadingPathSwitch(Default=@n9, Deferred=@n9, Forward=@n8, Mobile=@n8)
 ]
}
