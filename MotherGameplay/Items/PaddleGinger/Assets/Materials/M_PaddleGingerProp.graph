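# Translucent unlit material for a PaddleGinger prop, rendering it as a spinning, bobbing, scanlined hologram.
# Emissive blends desaturated, tinted diffuse with colored scanlines from object-relative height, weighted by Fresnel on rotation-corrected, normal-mapped normals.
# A sharper Fresnel and mask texture set opacity.
# World position offset adds axis rotation, sine bob and a FadeIn collapse toward the pivot; rim-color and secondary-line branches are left unconnected.
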Material /MotherGameplay/Items/PaddleGinger/Assets/Materials/M_PaddleGingerProp
graph {
 properties {
  bUseMaterialAttributes=true
  bUsedWithNanite=true
  blendMode=BLEND_Translucent
  pixelDepthOffsetMode=PDOM_Legacy
  shadingModel=MSM_Unlit
  usedShadingModels=Unlit
 }
 nodes [
  n1 = Subtract(A=ObjectPositionWS(), B=WorldPosition().XYZ)
  n2 = MaterialFunctionCall!(
   Number of Scanlines=ScalarParameter() {defaultValue=50, parameterName=Scanline Amount}
   Scanline Speed
   "Scanlines Thickness (Bias)"=ScalarParameter() {defaultValue=-0.15, parameterName=ScanlineSize}
   Scanlines Dark Value=ScalarParameter() {defaultValue=0.5, parameterName=Scanlines Dark}
   Scanlines Light Value=ScalarParameter() {defaultValue=1, parameterName=Scanlines Light}
   Block Opacity=Constant()
   UVs=Divide(A=ComponentMask(None=@n1) {b=true}, B=ScalarParameter() {defaultValue=85.518, parameterName=SizeSCanlineUV})
   Time=Multiply(A=Time(), B=ScalarParameter() {parameterName=ScanlinesTimeMultiplier})
   Block Repeat
  ) {materialFunction=/Game/Effects/Fort_Effects/Effects/Atlas/Materials/MF_Atlas_Scanlines_01}
  Add(A=@n2.Result, B=Multiply(A=@n2.Secondary Lines, B=ScalarParameter() {defaultValue=20, parameterName=SecondaryLineBirghtness}))
  Constant3Vector()
  n3 = Normalize(VectorInput=Transform(None=Constant3Vector() {constant=<rgba 0 0 1 0>}) {transformSourceType=TRANSFORMSOURCE_Local})
  n4 = Frac(None=Multiply(A=Time(), B=ScalarParameter() {defaultValue=0.125, parameterName=RotationSpeed}))
  n5 = TextureSampleParameter2D() {parameterName=Normals, samplerType=SAMPLERTYPE_Normal, texture=/Game/Environments/Props/Street/Textures/StreetProp_Construction_N}
  n6 = MaterialFunctionCall(
   BaseNormal=VertexInterpolator(VS=MaterialFunctionCall(Rotation Axis=@n3, Rotation Angle=@n4) {materialFunction=&FixRotateAboutAxisNormals}.New World Space Vertex Normal).PS
   AdditionalNormal=@n5.RGB
  ) {materialFunction=&BlendAngleCorrectedNormals}
  n7 = Fresnel(ExponentIn=ScalarParameter() {defaultValue=1.9445, parameterName=Fresnel}, Normal=@n6.Result) {baseReflectFraction=0}
  n8 = Desaturation(
   None=TextureSampleParameter2D() {parameterName=Diffuse, texture=/Game/Environments/Props/Street/Textures/StreetProp_Construction_D}.RGB
   Fraction=LinearInterpolate(A=ScalarParameter() {defaultValue=0.8, parameterName=Desat}, B=ScalarParameter() {defaultValue=1, parameterName=Desat2}, Alpha=@n7)
  )
  n9 = VectorParameter() {parameterName=DiffuseTint, defaultValue=<rgba 0.2267 0.8762 1.5 1>}
  n10 = Multiply(A=@n2.Result, B=LinearInterpolate(A=ScalarParameter() {defaultValue=0.05, parameterName=Min}, Alpha=@n7))
  n11 = Multiply(A=VectorParameter() {parameterName=ScanLineColor, defaultValue=<rgba 0 1.05 5 1>}.RGB, B=@n10)
  n12 = Add(A=Multiply(A=Multiply(A=@n8, B=@n9.RGB), B=OneMinus(None=@n10)), B=@n11)
  n13 = Saturate(None=Fresnel(ExponentIn=ScalarParameter() {defaultValue=8, parameterName=Fresnel2}, Normal=@n6.Result) {baseReflectFraction=0})
  LinearInterpolate(A=@n12, B=VectorParameter() {parameterName=ScanLineColorRim, defaultValue=<rgba 0 0.2545 0.4427 1>}.RGB, Alpha=@n13)
  n14 = OneMinus(None=@n2.Result)
  n15 = ScalarParameter() {defaultValue=0.4452, parameterName=ScanlineDiffuse}
  Multiply(A=@n11, B=Saturate(None=Add(A=@n14, B=@n15)))
  n16 = ScalarParameter() {parameterName=Fmask1}
  n17 = ScalarParameter() {parameterName=Fmask2}
  Saturate(None=Add(A=LinearInterpolate(A=@n16, B=@n17, Alpha=@n7), B=@n2.Result))
  n18 = Multiply(A=Time(), B=ScalarParameter() {defaultValue=0.2, parameterName=BobSpeed})
  n19 = AppendVector(A=Constant2Vector(), B=Multiply(A=Sine(None=@n18), B=ScalarParameter() {defaultValue=3, parameterName=BobAmount}))
  n20 = LinearInterpolate(A=Subtract(A=ObjectPositionWS(), B=WorldPosition().XYZ), B=Constant3Vector(), Alpha=ScalarParameter() {defaultValue=1, parameterName=FadeIn})
  n21 = RotateAboutAxis(NormalizedRotationAxis=@n3, RotationAngle=@n4, PivotPoint=ObjectPositionWS(), Position=Add(A=WorldPosition().XYZ, B=@n20))
  n22 = ScalarParameter() {defaultValue=0.78, parameterName=Opacity}
  n23 = ScalarParameter() {defaultValue=0.8, parameterName=Opacity2}
  n24 = TextureSampleParameter2D() {
   parameterName=MaskTexture
   samplerType=SAMPLERTYPE_Masks
   texture=/Game/Characters/Player/Female/Medium/Bodies/F_MED_Commando_01/Textures/SolidWhite_MaskCompression
  }
  SetMaterialAttributes!(
   MaterialAttributes=SetMaterialAttributes!(MaterialAttributes, Emissive Color=@n12) {attributeSetTypes [#1]}
   World Position Offset=Add(A=Add(A=@n19, B=@n21), B=@n20)
   Opacity=Multiply(A=LinearInterpolate(A=@n22, B=@n23, Alpha=@n13), B=@n24.R)
   Base Color
  ) {attributeSetTypes [#2, #3, #4]}
 ]
}
